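# Niagara system for the S11sr01 vehicle boost cosmetic, drawing faint, warm-tinted ribbon trails from the boost ports.
# One parent-derived emitter spawns a ribbon per port at a quality-scaled rate, launching particles backward at mode-specific speeds with skeletal positioning.
# Inherited vehicle movement and ribbon width fall off over life while color curves brighten the trail.
# A frontend-preview flag switches lifetime, offset, fade, velocity and color; the painted-variant override is disabled.
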NiagaraSystem /VehicleCosmetics/Boosts/Boost_S11sr01/FX/NS_S11sr01
dependencies {
 usedDataInterfaces [
  &NiagaraDataInterfaceCurve
  &NiagaraDataInterfaceColorCurve
  &NiagaraDataInterfaceSkeletalMesh
  &NiagaraDataInterfaceArrayFloat3
 ]
 usedDynamicInputs [&MakeVectorFromLinearColorRGB, &SelectVectorFromArray]
 usedModules [
  &self:SystemSpawnScript.SystemScriptSource.SystemScriptGraph.NiagaraNodeAssignment_3.SetVariables_#1
  &self:SystemSpawnScript.SystemScriptSource.SystemScriptGraph.NiagaraNodeAssignment_4.SetVariables_#2
  &self:SystemSpawnScript.SystemScriptSource.SystemScriptGraph.NiagaraNodeAssignment_2.SetVariables_#3
  &SystemState
 ]
}
emitters {NE_VC_Boost_Ribbon_Parent {
  emitterUpdateScript [
   &EmitterState()
   &VC_FilteredSocketCount()
   &VC_RateQualityHandler(Default Rate:NiagaraFloat=&Multiply_Float(
     A=100
     B=/VehicleCosmetics/FX/Niagara/Modules/VC_Mask_Floats(Coro Value:NiagaraFloat=1.2, FrontEnd Value:NiagaraFloat=1.2, FN Value:NiagaraFloat=0.5)
    ))
   &VC_SpawnRibbon_v1(Number of Ribbons=$Emitter.PortCount, SpawnRate=$Emitter.Spawn Rate)
   /SetVariables(Emitter.FadeFront:NiagaraFloat=&MakeCustomFloatFromBool(Bool=$User.bIsFrontendPreview, True Float=2, False Float=1)) {isSetParameters=true}
  ]
  particleSpawnScript [
   &VC_PaintedVariantsOverride() {enabled=false}
   &VC_InitializeRibbon_v1(
    Custom Offset=true
    Enable Skeletal Position=true
    Lifetime=&MakeCustomFloatFromBool(Bool=$User.bIsFrontendPreview, True Float=0.7, False Float=1)
    Local Vector Offset=<20 0 20>
    Offset Amount:NiagaraFloat=&MakeCustomFloatFromBool(Bool=$User.bIsFrontendPreview, True Float=0, False Float=1)
    Ribbon Width=80
   )
   &AddVelocity(Velocity=&VC_Mask_Vectors(
     Coro Value:Vector3f=<-1000 0 0>
     FrontEnd Value:Vector3f=&MakeCustomVectorFromBool(Bool=$User.bIsFrontendPreview, True Vector=<-2500 0 1>, False Vector=<-2000 0 0>)
     FN Value:Vector3f=<-1200 0 0>
    )) {module=AddVelocity001}
   /VehicleCosmetics/FX/VC_FXParents/Modules/DLR_ProductPreviews(GameModeID:NiagaraInt32=$System.GameModeID, bPreviewVehicleSpeed:NiagaraBool=$System.bPreviewVehicleSpeed)
  ]
  particleUpdateScript [
   &ParticleState()
   &InheritSourceMovement(
    Applied Position Scale=&VectorFromFloat(Value=&FloatFromCurve(FloatCurve=DataInterface() {propertyValues {Curve {
         defaultValue=1
         keys [
          {arriveTangent=-1.7885, interpMode=RCIM_Cubic, leaveTangent=-1.7885, tangentMode=RCTM_User, time=0.05, value=1}
          {interpMode=RCIM_Cubic, tangentMode=RCTM_User, time=1, value=0}
         ]
        }}}))
    ApplyForce=false
    ApplyPosition=true
    ApplyPositionFalloff=false
   )
   &SolveForcesAndVelocity(Clamp Velocity=false, Limit Acceleration=false)
   &VC_UpdateRibbon_v1(Enable Skeletal Position=true)
   &ScaleRibbonWidth(Ribbon Width Scale=&FloatFromCurve(FloatCurve=DataInterface() {propertyValues {Curve {keys [
         {arriveTangent=-0.1437, interpMode=RCIM_Cubic, leaveTangent=-0.1437, tangentMode=RCTM_User, time=0, value=1}
         {arriveTangent=-1.3699, interpMode=RCIM_Cubic, leaveTangent=-1.3699, tangentMode=RCTM_User, time=1, value=0.1}
        ]}}}))
   &ScaleColor(
    Color Scale:LinearColor=&MakeCustomLinearColorFromBool(
     Bool=$User.bIsFrontendPreview
     True Color=&ColorFromCurve(ColorCurve=DataInterface() {propertyValues {
        AlphaCurve {keys [<at 0 0.02 RCIM_Cubic>,<at 1 0.02 RCIM_Cubic>]}
        BlueCurve {keys [<at 0.3211 0.1441 RCIM_Cubic>,<at 0.8858 0.3357 RCIM_Cubic>]}
        GreenCurve {keys [<at 0.3211 0.162 RCIM_Cubic>,<at 0.8858 0.4268 RCIM_Cubic>]}
        RedCurve {keys [<at 0.3211 0.3005 RCIM_Cubic>,<at 0.8858 0.7 RCIM_Cubic>]}
       }})
     False Color=&ColorFromCurve(ColorCurve=DataInterface() {propertyValues {
        AlphaCurve {keys [<at 0 0.02>,<at 1 0.02>]}
        BlueCurve {keys [<at 0.0808 0.06>,<at 1 0.315>]}
        GreenCurve {keys [<at 0.0808 0.0787>,<at 1 0.4048>]}
        RedCurve {keys [<at 0.0808 0.2>,<at 1 0.7>]}
       }})
    )
    Scale Mode=3
   )
  ]
  rendererClasses [&NiagaraRibbonRendererProperties]
  renderers [{rendererClass=&NiagaraRibbonRendererProperties}]
 }}
summary {
 systemName=NS_S11sr01
 emitters [{emitterName=NE_VC_Boost_Ribbon_Parent, rendererClasses [&NiagaraRibbonRendererProperties]}]
 userVariables [
  User.bIsFrontendPreview {type=NiagaraBool}
  User.bIsPainted {type=NiagaraBool}
  User.blsIsCoroColor {type=NiagaraBool}
  User.bPreviewVehicleSpeed {type=NiagaraBool}
  User.ColorVariantID {type=NiagaraInt32}
  User.GameModeID {type=NiagaraInt32}
  User.PaintedColor {type=LinearColor, description="", defaultValue {struct=LinearColor, value=<rgba 1 1 1 1>}}
  User.PaintedScalar {type=NiagaraFloat, defaultValue=1, description=""}
  User.PortCount {type=NiagaraInt32, defaultValue=2}
 ]
}
